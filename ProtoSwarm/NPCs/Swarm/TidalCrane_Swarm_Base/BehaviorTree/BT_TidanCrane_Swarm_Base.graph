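# Base behavior tree for Tidal Crane swarm NPCs, choosing between combat movement around a goal actor and idling or wandering without one.
# A token query service feeds a melee-token check that splits attackers from non-holders, who loop zigzag and roar abilities.
# Token holders pause for airborne goals, then, by a range flag, jump or zigzag, or move to evaluator-supplied token positions, falling back to the goal actor.
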BehaviorTree /ProtoSwarm/NPCs/Swarm/TidalCrane_Swarm_Base/BehaviorTree/BT_TidanCrane_Swarm_Base
blackboard=&/BB_TidalCrane_Swarm_Base
root {
 name=Selector
 class=BTComposite_Selector
 children [
  Attack Or Hunt {
   class=BTComposite_Selector
   children [
    Sequence {
     class=BTComposite_Sequence
     children [
      Wait for Goal Pawn to Fall Off {
       class=BTTask_Wait
       decorators [
        Can Actor be considered Airborne {
         class=FortBTDecorator_CanBeConsideredAirborne
         properties {
          bInverseCondition=false
          flowAbortMode=None
          nodeName=Can Actor be considered Airborne
          blackboardKey {bNoneIsAllowedValue=false, selectedKeyName=GoalActor}
         }
        }
        Force Success {class=BTDecorator_ForceSuccess}
       ]
       properties {nodeName=Wait for Goal Pawn to Fall Off, randomDeviation {defaultValue=0.1}, waitTime {defaultValue=0.25}}
      }
      Selector {
       class=BTComposite_Selector
       children [
        Selector {
         class=BTComposite_Selector
         children [
          Jump {
           class=FortBTTask_ExecuteGameplayAbility
           decorators [
            Has Gameplay Ability {
             class=FortBTDecorator_GameplayAbility_HasGameplayAbility
             properties {gameplayAbilityTagContainer {defaultValue {gameplayTags [{tagName=Ability.ProtoSwarm.SmallWingedCreature.Jump}]}}}
            }
            Can Activate Gameplay Ability {
             class=FortBTDecorator_GameplayAbility_CanActivate
             properties {
              flowAbortMode=LowerPriority
              gameplayAbilityTagContainer {defaultValue {gameplayTags [{tagName=Ability.ProtoSwarm.SmallWingedCreature.Jump}]}}
              targetValue {key=None}
             }
            }
           ]
           properties {nodeName=Jump, gameplayAbilityTagContainer {defaultValue {gameplayTags [{tagName=Ability.ProtoSwarm.SmallWingedCreature.Jump}]}}}
          }
          Move in ZigZag {
           class=FortBTTask_ExecuteGameplayAbility
           decorators [
            Has Gameplay Ability {
             class=FortBTDecorator_GameplayAbility_HasGameplayAbility
             properties {gameplayAbilityTagContainer {defaultValue {gameplayTags [{tagName=Ability.ProtoSwarm.NonAttack.ZigZag}]}}}
            }
            Can Activate Gameplay Ability {
             class=FortBTDecorator_GameplayAbility_CanActivate
             properties {
              flowAbortMode=LowerPriority
              gameplayAbilityTagContainer {defaultValue {gameplayTags [{tagName=Ability.ProtoSwarm.NonAttack.ZigZag}]}}
              targetValue {key=None}
             }
            }
           ]
           properties {nodeName=Move in ZigZag, gameplayAbilityTagContainer {defaultValue {gameplayTags [{tagName=Ability.ProtoSwarm.NonAttack.ZigZag}]}}}
          }
          Wait {class=BTTask_Wait, properties {randomDeviation {defaultValue=0.25}, waitTime {defaultValue=0.5}}}
         ]
         decorators [Goal is far {
           class=BTDecorator_Blackboard
           properties {
            arithmeticOperation=Equal
            basicOperation=NotSet
            intValue=0
            nodeName=Goal is far
            notifyObserver=ValueChange
            stringValue=""
            blackboardKey {selectedKeyName=GoalActorDistance_IsInFarRange}
           }
          }]
         properties {
          children [
           {
            childComposite=null
            childTask=&self:FortBTTask_ExecuteGameplayAbility_5
            decorators [
             &self:FortBTDecorator_GameplayAbility_HasGameplayAbility_5
             &self:FortBTDecorator_GameplayAbility_CanActivate_5
            ]
           }
           {
            childComposite=null
            childTask=&self:FortBTTask_ExecuteGameplayAbility_4
            decorators [
             &self:FortBTDecorator_GameplayAbility_HasGameplayAbility_4
             &self:FortBTDecorator_GameplayAbility_CanActivate_4
            ]
           }
           {
            childComposite=null
            childTask=&self:BTTask_Wait_8
            decorators []
           }
          ]
          services []
         }
         services []
        }
        Selector {
         class=BTComposite_Selector
         children [
          Selector {
           class=BTComposite_Selector
           children [
            Move to Token Position {
             class=FortBTTask_GameMoveTo
             decorators [Near Goal Actor {
               class=BTDecorator_IsAtLocation
               properties {
                acceptableRadius=200
                bInverseCondition=false
                flowAbortMode=None
                geometricDistanceType=Distance3D
                nodeName=Near Goal Actor
                bPathFindingBasedTest {defaultValue=true, key=None}
                bUseNavAgentGoalLocation {defaultValue=false, key=None}
                blackboardKey {bNoneIsAllowedValue=false, selectedKeyName=GoalActor}
                parametrizedAcceptableRadius {
                 dataBinding=&self:BTDecorator_IsAtLocation_7.FortAIDataProvider_ScalableFloat_0
                 dataField=FloatValue
                 defaultValue=0
                }
               }
              }]
             properties {
              bDeimosFlavor=true
              bFinishMoveOnOverlap=false
              bStopAtGoal=true
              bUpdateMoveOnBlackboardKeyChanges=true
              nodeName=Move to Token Position
              acceptableRadius {defaultValue=20}
              bReachTestIncludesAgentRadius {defaultValue=false}
              bReachTestIncludesGoalRadius {defaultValue=false}
              blackboardKey {selectedKeyName=AIEvaluator_TokenPosition_Destination}
              filterClass {defaultValue=/ProtoSwarm/NavFilters/NavFilter_TidalCrane_Swarm.NavFilter_TidalCrane_Swarm_C}
              navFilterTag {tagName=None}
             }
            }
            Move to Token Position {
             class=FortBTTask_GameMoveTo
             properties {
              bDeimosFlavor=true
              bFinishMoveOnOverlap=false
              bStopAtGoal=true
              bUpdateMoveOnBlackboardKeyChanges=true
              nodeName=Move to Token Position
              acceptableRadius {defaultValue=20}
              bReachTestIncludesAgentRadius {defaultValue=false}
              bReachTestIncludesGoalRadius {defaultValue=false}
              blackboardKey {selectedKeyName=AIEvaluator_TokenPosition_Destination}
              filterClass {defaultValue=&NavFilter_TidalCrane_Swarm_C}
              navFilterTag {tagName=None}
             }
            }
           ]
           decorators [
            Have a Token Position ? {
             class=BTDecorator_Blackboard
             properties {
              arithmeticOperation=GreaterOrEqual
              intValue=3
              nodeName=Have a Token Position ?
              stringValue=ExecutionPending
              blackboardKey {selectedKeyName=AIEvaluator_TokenPosition_ExecutionStatus}
             }
            }
            Distance Between {
             class=FortBTDecorator_DistanceBetween
             properties {
              flowAbortMode=LowerPriority
              nodeName=Distance Between
              operator=Greater
              blackboardKeyB {selectedKeyName=AIEvaluator_TokenPosition_Destination}
              specifiedDistance {defaultValue=100}
             }
            }
           ]
           properties {
            children [
             {
              childComposite=null
              childTask=&self:FortBTTask_GameMoveTo_4
              decorators [&self:BTDecorator_IsAtLocation_7]
             }
             {
              childComposite=null
              childTask=&self:FortBTTask_GameMoveTo_2
              decorators []
             }
            ]
            services []
           }
           services []
          }
          Fallback Move to Goal Actor {
           class=FortBTTask_GameMoveTo
           decorators [Distance Between {
             class=FortBTDecorator_DistanceBetween
             properties {
              flowAbortMode=LowerPriority
              nodeName=Distance Between
              operator=Greater
              blackboardKeyB {selectedKeyName=GoalActor}
              specifiedDistance {defaultValue=200}
             }
            }]
           properties {
            bDeimosFlavor=true
            bFinishMoveOnOverlap=false
            bStopAtGoal=true
            nodeName=Fallback Move to Goal Actor
            acceptableRadius {defaultValue=100}
            bReachTestIncludesAgentRadius {defaultValue=false}
            bReachTestIncludesGoalRadius {defaultValue=false}
            blackboardKey {selectedKeyName=GoalActor}
            filterClass {defaultValue=&NavFilter_TidalCrane_Swarm_C}
            navFilterTag {tagName=None}
           }
          }
          Succeed {
           class=BTTask_FinishWithResult
           children []
           decorators [Can Actor be considered Airborne {
             class=FortBTDecorator_CanBeConsideredAirborne
             properties {
              bInverseCondition=false
              flowAbortMode=None
              nodeName=Can Actor be considered Airborne
              blackboardKey {bNoneIsAllowedValue=false, selectedKeyName=GoalActor}
             }
            }]
           properties {
            bIgnoreRestartSelf=false
            nodeName=Succeed
            result {defaultValue=0, enumType=/Script/AIModule.EBTNodeResult, key=None, nativeEnumTypeName=EBTNodeResult}
           }
           services []
          }
          Wait {class=BTTask_Wait, properties {waitTime {defaultValue=5}}}
         ]
         decorators [Goal is near {
           class=BTDecorator_Blackboard
           properties {
            arithmeticOperation=Equal
            flowAbortMode=Self
            intValue=0
            nodeName=Goal is near
            notifyObserver=ValueChange
            stringValue=""
            blackboardKey {selectedKeyName=GoalActorDistance_IsInFarRange}
           }
          }]
         properties {
          children [
           {
            childComposite=&self:BTComposite_Selector_6
            childTask=null
            decorators [
             &self:BTDecorator_Blackboard_13
             &self:FortBTDecorator_DistanceBetween_1
            ]
           }
           {
            childComposite=null
            childTask=&self:FortBTTask_GameMoveTo_1
            decorators [&self:FortBTDecorator_DistanceBetween_0]
           }
           {
            childComposite=null
            childTask=&self:BTTask_FinishWithResult_0
            decorators [&self:FortBTDecorator_CanBeConsideredAirborne_3]
           }
           {
            childComposite=null
            childTask=&self:BTTask_Wait_1
            decorators []
           }
          ]
          services []
         }
         services []
        }
       ]
       decorators [Loop {class=BTDecorator_Loop}]
       properties {
        children [
         {
          childComposite=&self:BTComposite_Selector_3
          childTask=null
          decorators [&self:BTDecorator_Blackboard_4]
         }
         {
          childComposite=&self:BTComposite_Selector_5
          childTask=null
          decorators [&self:BTDecorator_Blackboard_7]
         }
        ]
        services []
       }
       services []
      }
     ]
     decorators [Has Combat Token {
       class=FortBTDecorator_HasToken
       properties {
        bInverseCondition=false
        flowAbortMode=Both
        nodeName=Has Combat Token
        activationTagQuery {
         bCanEditDefaultValueType=false
         key=None
         defaultValue {
          _structType=/Script/GameplayTags.GameplayTagQuery
          autoDescription=""
          tokenStreamVersion=0
          userDescription=""
          queryTokenStream []
          tagDictionary []
         }
        }
        blackboardKey {bNoneIsAllowedValue=false, selectedKeyName=GoalActor}
        tokenQuery {
         bCanEditDefaultValueType=false
         key=None
         defaultValue {
          _structType=&GameplayTagQuery
          autoDescription=" ANY( Token.AI.BR.Melee )"
          tokenStreamVersion=0
          userDescription=""
          queryTokenStream [0,1,1,1,0]
          tagDictionary [{tagName=Token.AI.BR.Melee}]
         }
        }
       }
      }]
     properties {
      children [
       {
        childTask=&self:BTTask_Wait_12
        decorators [
         &self:FortBTDecorator_CanBeConsideredAirborne_2
         &self:BTDecorator_ForceSuccess_0
        ]
       }
       {
        childComposite=&self:BTComposite_Selector_4
        childTask=null
        decorators [&self:BTDecorator_Loop_3]
       }
      ]
      services [&self:FortAthenaBTService_AIEvaluator_0]
     }
     services [Token Position Evaluator {
       class=FortAthenaBTService_AIEvaluator
       properties {
        interval=0.5
        nodeName=Token Position Evaluator
        aIEvaluatorClass {defaultValue=&/EVAL_TidalCrane_Swarm_TokenPosition.EVAL_TidalCrane_Swarm_TokenPosition_C}
       }
      }]
    }
    Sequence {
     class=BTComposite_Sequence
     children [
      Move in ZigZag {
       class=FortBTTask_ExecuteGameplayAbility
       decorators [
        Has Gameplay Ability {
         class=FortBTDecorator_GameplayAbility_HasGameplayAbility
         properties {gameplayAbilityTagContainer {defaultValue {gameplayTags [{tagName=Ability.ProtoSwarm.NonAttack.ZigZag}]}}}
        }
        Can Activate Gameplay Ability {
         class=FortBTDecorator_GameplayAbility_CanActivate
         properties {
          flowAbortMode=None
          gameplayAbilityTagContainer {defaultValue {gameplayTags [{tagName=Ability.ProtoSwarm.NonAttack.ZigZag}]}}
          targetValue {key=None}
         }
        }
       ]
       properties {nodeName=Move in ZigZag, gameplayAbilityTagContainer {defaultValue {gameplayTags [{tagName=Ability.ProtoSwarm.NonAttack.ZigZag}]}}}
      }
      Wait {class=BTTask_Wait, properties {randomDeviation {defaultValue=0.25}, waitTime {defaultValue=1.5}}}
      Roar {
       class=FortBTTask_ExecuteGameplayAbility
       decorators [
        Has Gameplay Ability {
         class=FortBTDecorator_GameplayAbility_HasGameplayAbility
         properties {gameplayAbilityTagContainer {defaultValue {gameplayTags [{tagName=Ability.ProtoSwarm.NonAttack.Roar}]}}}
        }
        Can Activate Gameplay Ability {
         class=FortBTDecorator_GameplayAbility_CanActivate
         properties {
          flowAbortMode=None
          gameplayAbilityTagContainer {defaultValue {gameplayTags [{tagName=Ability.ProtoSwarm.NonAttack.Roar}]}}
          targetValue {key=None}
         }
        }
       ]
       properties {nodeName=Roar, gameplayAbilityTagContainer {defaultValue {gameplayTags [{tagName=Ability.ProtoSwarm.NonAttack.Roar}]}}}
      }
     ]
     decorators [Loop {class=BTDecorator_Loop}]
     properties {children [
       {
        childTask=&self:FortBTTask_ExecuteGameplayAbility_3
        decorators [
         &self:FortBTDecorator_GameplayAbility_HasGameplayAbility_1
         &self:FortBTDecorator_GameplayAbility_CanActivate_1
        ]
       }
       {childTask=&self:BTTask_Wait_3}
       {
        childTask=&self:FortBTTask_ExecuteGameplayAbility_1
        decorators [
         &self:FortBTDecorator_GameplayAbility_HasGameplayAbility_3
         &self:FortBTDecorator_GameplayAbility_CanActivate_3
        ]
       }
      ]}
    }
   ]
   decorators [Has Goal Actor? {
     class=BTDecorator_Blackboard
     properties {
      arithmeticOperation=Equal
      intValue=0
      nodeName=Has Goal Actor?
      notifyObserver=ValueChange
      stringValue=""
      blackboardKey {selectedKeyName=GoalActor}
     }
    }]
   properties {
    nodeName=Attack Or Hunt
    children [
     {
      childComposite=&self:BTComposite_Sequence_2
      childTask=null
      decorators [&self:FortBTDecorator_HasToken_1]
     }
     {
      childComposite=&self:BTComposite_Sequence_1
      childTask=null
      decorators [&self:BTDecorator_Loop_2]
     }
    ]
    services [&self:FortAthenaBTService_TokenQuery_2]
   }
   services [Executing...Token Query {
     class=FortAthenaBTService_TokenQuery
     properties {
      bCallTickOnSearchStart=false
      bRestartTimerOnEachActivation=false
      interval=99999
      nodeName=Executing...Token Query
      randomDeviation=0
      targetActorName=GoalActor
      weaponKeyName=AIEvaluator_Global_DesiredWeapon
      activationTagQuery {
       bCanEditDefaultValueType=false
       key=None
       defaultValue {
        _structType=&GameplayTagQuery
        autoDescription=""
        tokenStreamVersion=0
        userDescription=""
        queryTokenStream []
        tagDictionary []
       }
      }
      becomeRelevantTokenAction {
       defaultValue=1
       enumType=/Script/FortniteAIServer.EFortAthenaBTServiceTokenQueryAction
       key=None
       nativeEnumTypeName=EFortAthenaBTServiceTokenQueryAction
      }
      ceaseRelevantTokenAction {
       defaultValue=2
       enumType=&EFortAthenaBTServiceTokenQueryAction
       key=None
       nativeEnumTypeName=EFortAthenaBTServiceTokenQueryAction
      }
      defaultTokenSystemTagQuery {
       bCanEditDefaultValueType=false
       key=None
       defaultValue {
        _structType=&GameplayTagQuery
        autoDescription=" ANY( Token.AI.BR.Melee )"
        tokenStreamVersion=0
        userDescription=""
        queryTokenStream [0,1,1,1,0]
        tagDictionary [{tagName=Token.AI.BR.Melee}]
       }
      }
      equippedItemTagAssociationDatas []
      targetChangeTokenAction {
       defaultValue=1
       enumType=&EFortAthenaBTServiceTokenQueryAction
       key=None
       nativeEnumTypeName=EFortAthenaBTServiceTokenQueryAction
      }
      tokenReleaseDelay {defaultValue=0, key=None}
      weaponChangeTokenAction {
       defaultValue=1
       enumType=&EFortAthenaBTServiceTokenQueryAction
       key=None
       nativeEnumTypeName=EFortAthenaBTServiceTokenQueryAction
      }
     }
    }]
  }
  Selector {
   class=BTComposite_Selector
   children [
    GameplayAbility {
     class=FortBTTask_ExecuteGameplayAbility
     decorators [
      Gameplay Tag Query {
       class=BTDecorator_GameplayTagQuery
       properties {
        flowAbortMode=None
        nodeName=Gameplay Tag Query
        gameplayTagQuery {
         autoDescription=" ANY( NPC.ProtoSwarm.SmallCreature.AlwaysOn )"
         queryTokenStream [0,1,1,1,0]
         tagDictionary [{tagName=NPC.ProtoSwarm.SmallCreature.AlwaysOn}]
        }
       }
      }
      Has Gameplay Ability {
       class=FortBTDecorator_GameplayAbility_HasGameplayAbility
       properties {gameplayAbilityTagContainer {defaultValue {gameplayTags [{tagName=NPC.Ability.NonAttack.Wander.Freely}]}}}
      }
      Can Activate Gameplay Ability {
       class=FortBTDecorator_GameplayAbility_CanActivate
       properties {
        flowAbortMode=Both
        gameplayAbilityTagContainer {defaultValue {gameplayTags [{tagName=NPC.Ability.NonAttack.Wander.Freely}]}}
        targetValue {key=None}
       }
      }
     ]
     properties {nodeName=GameplayAbility, gameplayAbilityTagContainer {defaultValue {gameplayTags [{tagName=NPC.Ability.NonAttack.Wander.Freely}]}}}
    }
    Wait {class=BTTask_Wait, properties {waitTime {defaultValue=9999}}}
   ]
   decorators [
    Has no Goal Actor? {
     class=BTDecorator_Blackboard
     properties {
      arithmeticOperation=Equal
      basicOperation=NotSet
      intValue=0
      nodeName=Has no Goal Actor?
      notifyObserver=ValueChange
      stringValue=""
      blackboardKey {selectedKeyName=GoalActor}
     }
    }
    Loop {class=BTDecorator_Loop}
   ]
   properties {
    children [
     {
      childComposite=null
      childTask=&self:FortBTTask_ExecuteGameplayAbility_0
      decorators [
       &self:BTDecorator_GameplayTagQuery_0
       &self:FortBTDecorator_GameplayAbility_HasGameplayAbility_0
       &self:FortBTDecorator_GameplayAbility_CanActivate_0
      ]
     }
     {
      childComposite=null
      childTask=&self:BTTask_Wait_0
      decorators []
     }
    ]
    services []
   }
   services []
  }
 ]
 decorators []
 properties {
  children [
   {
    childComposite=&self:BTComposite_Selector_48
    childTask=null
    decorators [&self:BTDecorator_Blackboard_122]
   }
   {
    childComposite=&self:BTComposite_Selector_1
    childTask=null
    decorators [
     &self:BTDecorator_Blackboard_0
     &self:BTDecorator_Loop_0
    ]
   }
  ]
  services []
 }
 services []
}
